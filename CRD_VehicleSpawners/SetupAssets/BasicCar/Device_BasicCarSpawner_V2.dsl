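; Basic car spawner device, version 2, a child Blueprint that exposes its Color and Style user option as an integer index.
; The getter converts the EColorBasicCar value to an integer and subtracts one, while the setter adds one and casts back to the enum, so the index is offset from the raw enum value.
; The construction script only calls the parent.
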(fn ColorAndStyleAsAnInteger
  (return (- (Math|Conversions|ToInteger(Byte) (Variables|UserOptions|Basic|GetColor%20and%20Style)) 1)))

(fn SetColorOption
  (params (ColorIndex :type Integer))
  (Variables|UserOptions|Basic|SetColor%20and%20Style (Utilities|Enum|BytetoEnumEColorBasicCar (Math|Conversions|ToByte(Integer) (+ ColorIndex 1)))))

(fn ConstructionScript
  (|CallParentFunction))
